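# Physics asset for the template glider skeletal mesh of the VinderTech GliderChute gadget.
# It holds a single capsule body on the SubFrame bone, offset and tilted to approximate the glider's bulk.
# With no constraints, it gives one simple collision and bounds volume rather than an articulated ragdoll.
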PhysicsAsset /Game/Gadgets/Assets/VinderTech_GliderChute/Glider_Template/Meshes/SK_Glider_Template_PhysicsAsset
bodies {SubFrame [{shapeType=Capsule, length=111.2166, radius=72.3303, center=<0.0088 -24.8812 23.1535>, rotation=<rot 18.6857 -89.9691 -90>}]}
constraints []
